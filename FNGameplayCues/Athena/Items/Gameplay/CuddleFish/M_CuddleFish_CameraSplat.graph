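# Translucent unlit particle material for the CuddleFish camera splat effect, drawing dark splats with iridescent color.
# Dynamic parameters pick a splat atlas cell and scale normal-map UV distortion.
# Scrolling swirl-distorted noise indexes the iridescent gradient, tinted by particle color and masked by splat green, while splat normals fake a highlight against a fixed view-space direction.
# Particle alpha drives opacity erosion through the splat channels and a sphere mask, also flattening normals.
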Material /FNGameplayCues/Athena/Items/Gameplay/CuddleFish/M_CuddleFish_CameraSplat
graph {
 properties {
  bUsedWithParticleSprites=true
  blendMode=BLEND_Translucent
  pixelDepthOffsetMode=PDOM_Legacy
  shadingModel=MSM_Unlit
  usedShadingModels=Unlit
 }
 nodes [
  Constant() {r=1}
  n1 = VectorParameter() {parameterName=DecalColorA, defaultValue=<rgba 0.0046 0.0035 0.0052 0.3>}
  n2 = Add(A=Multiply(A=Time(), B=ScalarParameter() {defaultValue=0.01, parameterName=Color Speed 2}), B=TextureCoordinate())
  n3 = Panner(Speed=ScalarParameter() {defaultValue=0.03, parameterName=Color Speed 3}) {speedY=0.1}
  n4 = ComponentMask(None=Multiply(
    A=TextureSample(UVs=@n3) {samplerType=SAMPLERTYPE_LinearColor, texture=/Game/Global/Textures/Default/Flowmaps/T_SwirlFlow_02}.RGB
    B=ScalarParameter() {defaultValue=0.5, parameterName=Distortion Strength}
   )) {g=true, r=true}
  n5 = TextureSample(UVs=Multiply(A=Add(A=@n2, B=@n4), B=ScalarParameter() {defaultValue=1, parameterName=Iridescence Tiling})) {texture=&2ndInfectedMapLinear}
  n6 = TextureObjectParameter() {parameterName=Iridescent Texture, texture=/Game/Weapons/WeaponSkins/Wraps/Textures/Iridescence/T_iridescent_gradient}
  n7 = Multiply(
   A=Desaturation(
    None=TextureSample(UVs=@n5.R, Tex=@n6) {texture=&T_iridescent_gradient}.RGB
    Fraction=ScalarParameter() {parameterName=Iridescence Desaturation Fraction}
   )
   B=VectorParameter() {parameterName=Iridescence, defaultValue=<rgba 0.1175 0.2031 0.1113 0>}.RGB
  )
  n8 = Multiply(A=ParticleColor().RGB, B=@n7)
  n9 = TextureSample() {samplerType=SAMPLERTYPE_Normal, texture=/Game/Global/Textures/Default/Tileable_Patterns/1Normal}
  n10 = Multiply(
   A=AppendVector(A=@n9.R, B=@n9.R)
   B=DynamicParameter() {defaultValue=<rgba 1 1 0 1>, paramNames [IndexU,IndexV,Distortion,Param4]}.Distortion
  )
  n11 = TextureCoordinate()
  n12 = AppendVector(A=ScalarParameter() {defaultValue=4, parameterName=SubdivisionsU}, B=ScalarParameter() {defaultValue=1, parameterName=SubdivisionsV})
  n13 = DynamicParameter() {defaultValue=<rgba 1 1 0 1>, paramNames [IndexU,IndexV,Distortion,Param4]}
  n14 = Floor(None=@n13.IndexU)
  n15 = Floor(None=@n13.IndexV)
  n16 = Add(A=Divide(A=Add(A=@n10, B=@n11), B=@n12), B=Divide(A=AppendVector(A=@n14, B=@n15), B=@n12))
  n17 = TextureSample(UVs=@n16) {samplerType=SAMPLERTYPE_LinearColor, texture=/Game/Athena/Items/Weapons/Prototype/Papaya/T_PaintSplats_01}
  n18 = Power(Base=@n17.G, Exp=ScalarParameter() {defaultValue=3, parameterName=Iridescence Power})
  n19 = TextureSample(UVs=@n16) {samplerType=SAMPLERTYPE_Normal, texture=/Game/Athena/Items/Weapons/Prototype/Papaya/T_PaintSplats_Normal}
  n20 = Multiply(A=@n19.RGB, B=VectorParameter() {parameterName=NormalOuterStength, defaultValue=<rgba 0.1 0.1 1 0>}.RGB)
  n21 = OneMinus(None=ParticleColor().A)
  n22 = DotProduct(
   A=Transform(None=LinearInterpolate(A=@n19.RGB, B=@n20, Alpha=@n21))
   B=Transform(None=Constant3Vector() {constant=<rgba -1 1 0 0>}) {transformSourceType=TRANSFORMSOURCE_View}
  )
  n23 = Multiply(A=Add(A=@n7, B=ScalarParameter() {defaultValue=0.3, parameterName=Highlight Add}), B=Power(Base=Saturate(None=@n22)) {constExponent=6})
  n24 = ScalarParameter() {defaultValue=0.6, parameterName=Highlight Color Strength}
  MaterialFunctionCall(Emissive=Add(A=LinearInterpolate(A=@n1.RGB, B=@n8, Alpha=@n18), B=Multiply(A=@n23, B=@n24))) {materialFunction=&MF_FX_EmissiveAdjustments}
  n25 = Add(A=@n17.G, B=SphereMask(A=@n17.B, B=Constant(), Radius=@n21, Hardness=ScalarParameter() {defaultValue=0.2, parameterName=ErosionHardness}))
  n26 = Multiply(A=@n21) {constB=2}
  Multiply(A=Subtract(A=@n17.R, B=Multiply(A=@n25, B=@n26)), B=ScalarParameter() {defaultValue=0.98, parameterName=Max Opacity})
  ParticleRelativeTime()
 ]
}
